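# Material function for the gold Superlevel Galaxy player cosmetic, blending an animated galaxy look over input material attributes by an Alpha input.
# Emissive combines a pre-skinned height gradient, drifting stars and a two-color edge glow, driven by Fresnel and Edges inputs.
# Stars and galaxy tint use rotating camera-vector triplanar projection, with tiling adjusted for camera FOV.
# A second output exposes a scrolling world-aligned smoke mask as "Fresnel panner".
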MaterialFunction /BRCosmetics/Characters/Player/Common/MaterialFunctions/Superlevel_Galaxy/MF_Superlevel_Galaxy_Gold
graph {nodes [
  Add()
  n1 = FunctionInput() {inputName=Fresnel, inputType=FunctionInput_Scalar}
  n2 = Power(Base=@n1, Exp=ScalarParameter() {defaultValue=12, group=Galaxy pink, parameterName=EdgeFresnel Power})
  n3 = Time()
  n4 = Sine(None=Multiply(A=Multiply(A=@n3) {constB=0.001}, B=ScalarParameter() {defaultValue=50, parameterName=Sine Edge Speed}))
  Add(A=@n2, B=Saturate(None=ConstantBiasScale(None=@n4) {bias=0.5, scale=1}))
  Clamp()
  n5 = Constant() {r=0.1}
  Clamp(None=@n2, Min=@n5)
  ComponentMask() {g=true, r=true}
  Constant()
  Constant()
  n6 = FunctionInput() {inputName=Material, inputType=FunctionInput_MaterialAttributes}
  n7 = NamedRerouteUsage()
  n8 = FunctionInput() {inputName=Pre Skinned Position}
  n9 = ComponentMask(None=@n8) {b=true}
  n10 = Saturate(None=Power(Base=Divide(A=@n9) {constB=125}, Exp=ScalarParameter() {defaultValue=0.75, parameterName=GradPower}))
  n11 = ScalarParameter() {defaultValue=1.5, group=Galaxy pink, parameterName=BG Color Fresnel}
  n12 = Multiply(
   A=LinearInterpolate(
    A=VectorParameter() {parameterName=Galaxy_BG_Color_Bottom, defaultValue=<rgba 0.0156 0.0046 0.0154 1>}.RGB
    B=VectorParameter() {parameterName=Galaxy_BG_Color, defaultValue=<rgba 0.0201 0.0516 0.0521 1>}.RGB
    Alpha=@n10
   )
   B=Clamp(None=Power(Base=@n1, Exp=@n11) {constExponent=0.5}) {maxDefault=1.3, minDefault=0.66}
  )
  n13 = Add(A=FunctionInput() {inputName=Edges, inputType=FunctionInput_Scalar}, B=Multiply(A=@n2, B=Constant() {r=0.25}))
  n14 = SetMaterialAttributes!(
   MaterialAttributes=GetMaterialAttributes(None=@n6) {attributeGetTypes [#1, #2]}.MaterialAttributes
   BaseColor=Constant()
   Metallic=Constant()
   EmissiveColor=LinearInterpolate(A=Add(A=@n12, B=@n7), B=NamedRerouteUsage(), Alpha=Saturate(None=@n13))
   Roughness=ScalarParameter() {defaultValue=0.25, parameterName=GalaxyRough}
   Specular=ScalarParameter() {parameterName=GalaxySpec}
  ) {attributeSetTypes [
    #1
    #2
    #3
    #4
    #5
   ]}
  FunctionOutput(None=BlendMaterialAttributes(A=@n6, B=@n14, Alpha=FunctionInput() {inputName=Alpha, inputType=FunctionInput_Scalar}))
  n15 = NamedRerouteUsage()
  n16 = Time()
  n17 = Multiply(A=Multiply(A=@n16) {constB=0.001}, B=ScalarParameter() {defaultValue=1, parameterName=SpeedFresneh})
  n18 = ComponentMask(None=MaterialFunctionCall(
    TextureObject=TextureObject() {texture=&T_Inky_Smoke_Tile}
    TextureSize=ScalarParameter() {defaultValue=50, parameterName=Flame Tiling}
    WorldPosition=Add(A=@n15, B=@n17)
   ) {materialFunction=/Game/Weapons/WeaponSkins/Wraps/MaterialFunctions/WorldAlignedTexture}.XYZ Texture) {g=true}
  FunctionOutput(None=NamedRerouteDeclaration(None=LinearInterpolate(Alpha=@n18) {constA=0.33}) {name=Fresnel Panner, nodeColor=<rgba 1 0 0.6797 1>}) {outputName=Fresnel panner}
  n19 = Time()
  n20 = MaterialFunctionCall() {materialFunction=/Engine/Functions/Engine_MaterialFunctions02/Utility/FOV}
  n21 = Constant() {r=0.414}
  n22 = Constant() {r=1.091}
  n23 = Constant()
  n24 = Constant() {r=1}
  n25 = LinearInterpolate(
   A=Constant() {r=3}
   B=Constant() {r=1}
   Alpha=Saturate(None=MaterialFunctionCall(Input=@n20."FOV ", Input Low=@n21, Input High=@n22, Target Low=@n23, Target High=@n24) {materialFunction=&RemapValueRange}.Result)
  )
  n26 = MaterialFunctionCall!(
   CameraVector
   Rotation Speed=Multiply(A=ScalarParameter() {defaultValue=1, parameterName=GalaxyRotSpeed}, B=Multiply(B=@n19) {constA=0.1})
   Texture=TextureObject() {texture=&/GalaxyTintMask_M}
   Tiling=Multiply(A=ScalarParameter() {defaultValue=3, group=Galaxy pink, parameterName=Tiling Stars 2}, B=@n25)
   Offset=Multiply(A=@n19, B=VectorParameter() {group=Galaxy pink, parameterName=Pannings Speed Stars 2, defaultValue=<rgba -0.02 0.05 0.1 1>}.RGB)
   AxisFadeContrast
  ) {materialFunction=/Game/Weapons/WeaponSkins/MaterialFunction/TriplanarCameraVector_Rotated}
  MaterialFunctionCall(Float3=@n26.Result) {materialFunction=&BreakOutFloat3Components}
  Multiply(A=@n10)
  Multiply(A=@n16)
  n27 = ScalarParameter() {defaultValue=20, parameterName=Edge Glow Brightness}
  n28 = Power(Base=@n1, Exp=ScalarParameter() {defaultValue=2.5, parameterName=Edge Glow Fresnel}) {constExponent=0.5}
  n29 = VectorParameter() {parameterName=Galaxy_BG_Edge, defaultValue=<rgba 0.2467 0.1117 0.7396 1>}
  n30 = VectorParameter() {parameterName=Galaxy_BG_Edge2, defaultValue=<rgba 0.7991 0.3176 1 1>}
  n31 = NamedRerouteUsage()
  NamedRerouteDeclaration(None=Multiply(A=LinearInterpolate(A=@n5, B=@n27, Alpha=@n28) {constB=25}, B=LinearInterpolate(A=@n29.RGB, B=@n30.RGB, Alpha=@n31))) {name=Edge Glow, nodeColor=<rgba 0.5703 0 1 1>}
  NamedRerouteDeclaration(None=@n8) {name=Pre Skinned World Position, nodeColor=<rgba 0.5234 1 0 1>}
  NamedRerouteDeclaration(None=@n10) {name=TopBottomGrad, nodeColor=<rgba 1 0.6797 0 1>}
  n32 = NamedRerouteUsage()
  n33 = OneMinus(None=Saturate(None=Power(Base=@n32) {constExponent=10}))
  n34 = TextureObject() {samplerType=SAMPLERTYPE_LinearColor, texture=&/SL_Galaxy_Stars}
  n35 = MaterialFunctionCall(
   Rotation Speed=Constant() {r=0.02}
   Texture=@n34
   Tiling=Multiply(A=ScalarParameter() {defaultValue=2.25, parameterName=Tiling Stars 1}, B=@n25)
   Offset=Multiply(A=@n19, B=VectorParameter() {parameterName=Pannings Speed Stars 1, defaultValue=<rgba 0.02 0.05 0.1 1>}.RGB)
  ) {materialFunction=&TriplanarCameraVector_Rotated}
  n36 = Multiply(A=NamedRerouteUsage(), B=ComponentMask(None=@n35.Result) {g=true})
  n37 = Multiply(A=Saturate(None=Multiply(A=@n33, B=@n36)), B=VectorParameter() {parameterName=Stars Color 1, defaultValue=<rgba 4 2.599 1.6137 1>}.RGB)
  n38 = Multiply(A=VectorParameter() {parameterName=Stars Color 2, defaultValue=<rgba 1.885 2.6348 3 1>}.RGB, B=ComponentMask(None=@n35.Result) {b=true})
  n39 = ScalarParameter() {defaultValue=1, group=Galaxy pink, parameterName=Panner Galaxy Brightness Max}
  NamedRerouteDeclaration(None=Add(A=Add(A=@n37, B=@n38), B=Multiply(A=@n39, B=@n26.Result))) {name=Stars, nodeColor=<rgba 0 1 0.9062 1>}
  OneMinus()
 ]}
