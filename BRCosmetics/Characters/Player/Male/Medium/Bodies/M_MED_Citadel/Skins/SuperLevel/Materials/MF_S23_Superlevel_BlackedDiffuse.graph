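# Material function for the Season 23 Super Level skin of the Citadel character that creates a "blacked diffuse" variant of a surface and blends it in.
# It desaturates the Original input's base color and tints it by metallic, dark for dielectrics and bright for metals, then clamps it.
# Roughness gets separate dielectric and metal multipliers, and emissive is set to a constant.
# An Alpha input blends a second material attributes input toward this result.
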MaterialFunction /BRCosmetics/Characters/Player/Male/Medium/Bodies/M_MED_Citadel/Skins/SuperLevel/Materials/MF_S23_Superlevel_BlackedDiffuse
graph {nodes [
  Constant()
  n1 = GetMaterialAttributes(None=FunctionInput() {inputName=Original, inputType=FunctionInput_MaterialAttributes}) {attributeGetTypes [
    #1
    #2
    #3
    #4
   ]}
  n2 = Multiply(
   A=Desaturation(None=@n1.BaseColor, Fraction=ScalarParameter() {defaultValue=0.9, parameterName=Diffuse Desaturation})
   B=LinearInterpolate(
    A=VectorParameter() {parameterName=BlackedDialectricTint, defaultValue=<rgba 0.0469 0.0469 0.0469 1>}.RGB
    B=VectorParameter() {parameterName=BlackedMetalTint, defaultValue=<rgba 1.374 1.5 1.485 1>}.RGB
    Alpha=@n1.Metallic
   )
  )
  n3 = ScalarParameter() {defaultValue=1.1, parameterName=Dialectric Roughness Adjust}
  n4 = ScalarParameter() {defaultValue=0.6, parameterName=Metal Roughness Adjust}
  n5 = SetMaterialAttributes!(
   MaterialAttributes=@n1.MaterialAttributes
   BaseColor=Clamp(None=Saturate(None=@n2)) {minDefault=0.02}
   Roughness=Multiply(A=@n1.Roughness, B=LinearInterpolate(A=@n3, B=@n4, Alpha=@n1.Metallic)) {constB=1.2}
   Specular=@n1.MaterialAttributes
   Metallic=@n1.MaterialAttributes
   EmissiveColor=Constant()
  ) {attributeSetTypes [
    #1
    #4
    #3
    #2
    #5
   ]}
  FunctionOutput(None=BlendMaterialAttributes(
    A=FunctionInput() {inputType=FunctionInput_MaterialAttributes}
    B=@n5
    Alpha=FunctionInput() {inputName=Alpha, inputType=FunctionInput_Scalar}
   ))
 ]}
