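# Material function for the Season 23 Citadel SuperLevel skin that layers a glowing emissive look over incoming material attributes.
# It lerps two SL Emissive Color parameters by Alpha Secondary and scales the result by an inverted Fresnel input raised to a power parameter.
# That glow replaces EmissiveColor alongside constant base color, roughness, specular and metallic overrides.
# The override blends over the original attributes using the max of Alpha Prime and Alpha Secondary.
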MaterialFunction /BRCosmetics/Characters/Player/Male/Medium/Bodies/M_MED_Citadel/Skins/SuperLevel/Materials/MF_S23_Superlevel_Emissives
graph {nodes [
  Constant()
  n1 = FunctionInput() {inputType=FunctionInput_MaterialAttributes}
  n2 = VectorParameter() {parameterName=SL Emissive Color, defaultValue=<rgba 5 0 4.9148 1>}
  n3 = VectorParameter() {parameterName=SL Emissive Color 2, defaultValue=<rgba 5 0 4.9148 1>}
  n4 = FunctionInput() {inputName=Alpha Secondary, inputType=FunctionInput_Scalar}
  n5 = OneMinus(None=FunctionInput() {inputName=Fresnel, inputType=FunctionInput_Scalar})
  n6 = ScalarParameter() {defaultValue=5, parameterName=SLEmissiveFresnelPower}
  n7 = SetMaterialAttributes!(
   MaterialAttributes=GetMaterialAttributes(None=@n1).MaterialAttributes
   BaseColor=Constant()
   Roughness=Constant() {r=0.15}
   Specular=Constant() {r=0.75}
   Metallic=Constant()
   EmissiveColor=Multiply(A=LinearInterpolate(A=@n2.RGB, B=@n3.RGB, Alpha=@n4), B=Power(Base=@n5, Exp=@n6))
  ) {attributeSetTypes [
    #1
    #2
    #3
    #4
    #5
   ]}
  n8 = FunctionInput() {inputName=Alpha Prime, inputType=FunctionInput_Scalar}
  FunctionOutput(None=BlendMaterialAttributes(A=@n1, B=@n7, Alpha=Max(A=@n8, B=@n4)))
 ]}
