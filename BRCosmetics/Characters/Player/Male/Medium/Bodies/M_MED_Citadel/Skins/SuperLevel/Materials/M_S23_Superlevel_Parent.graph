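# Parent character material for the S23 SuperLevel skin on the Citadel body, giving it a gold, pink or black galaxy look.
# Static switches isGold and isPink choose the galaxy function, which takes pre-skinned position, Fresnel and an inverted skin mask.
# Superlevel_Mask drives the following emissive (G) and blacked-diffuse (B) functions, plus a tunable edge mask (A).
# Shared character PreFX and PostFX functions bracket the chain using material attributes.
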Material /BRCosmetics/Characters/Player/Male/Medium/Bodies/M_MED_Citadel/Skins/SuperLevel/Materials/M_S23_Superlevel_Parent
graph {
 properties {
  bUseMaterialAttributes=true
  bUsedWithInstancedStaticMeshes=true
  bUsedWithMeshDeformer=true
  bUsedWithMorphTargets=true
  bUsedWithSkeletalMesh=true
  pixelDepthOffsetMode=PDOM_Legacy
  usedShadingModels=Default Lit
 }
 nodes [
  n1 = MaterialFunctionCall() {materialFunction=&BaseFN_CharacterFunction_PreFX}
  n2 = TextureSampleParameter2D() {
   parameterName=Superlevel_Mask
   samplerType=SAMPLERTYPE_Masks
   texture=/BRCosmetics/Characters/Player/Male/Medium/Bodies/M_MED_Venice/Skins/SL/M_MED_Venice_SuperLevel_Body_FX
  }
  n3 = NamedRerouteUsage()
  n4 = Saturate(None=Subtract(A=Constant() {r=1}, B=@n1.Skin))
  n5 = MaterialFunctionCall!(Edges=@n3, Pre Skinned Position=@n1.Pre skinned Local Position, Alpha=@n4, Fresnel=@n1.Fresnel, Material=@n1.Result) {materialFunction=/BRCosmetics/Characters/Player/Common/MaterialFunctions/Superlevel_Galaxy/MF_Superlevel_Galaxy_Gold}
  n6 = StaticSwitchParameter(
   True=MaterialFunctionCall!(Edges=@n3, Pre Skinned Position=@n1.Pre skinned Local Position, Alpha=@n4, Fresnel=@n1.Fresnel, Material=@n1.Result) {materialFunction=/BRCosmetics/Characters/Player/Common/MaterialFunctions/Superlevel_Galaxy/MF_Superlevel_Galaxy_Pink}.Result
   False=MaterialFunctionCall!(Edges=@n3, Pre Skinned Position=@n1.Pre skinned Local Position, Alpha=@n4, Fresnel=@n1.Fresnel, Material=@n1.Pre skinned Local Position) {materialFunction=/BRCosmetics/Characters/Player/Common/MaterialFunctions/Superlevel_Galaxy/MF_Superlevel_Galaxy_Black}.Result
  ) {parameterName=isPink}
  n7 = MaterialFunctionCall!(
   Alpha=@n2.B
   Original=@n1.Result
   In=MaterialFunctionCall!(Alpha Prime=@n2.G, Fresnel=@n1.Fresnel, Alpha Secondary=@n2.G, In=StaticSwitchParameter(True=@n5.Result, False=@n6) {parameterName=isGold}) {materialFunction=&/MF_S23_Superlevel_Emissives}.Result
  ) {materialFunction=&/MF_S23_Superlevel_BlackedDiffuse}
  MaterialFunctionCall(AO=@n1.AO, Distance Blend=@n1.AO, Fresnel=@n1.AO, Gmap=@n1.AO, Pre Skinned Local Position=@n1.AO, Material=@n7.Result) {materialFunction=&BaseFN_CharacterFunction_PostFX}
  Multiply(A=NamedRerouteUsage(), B=NamedRerouteUsage())
  NamedRerouteDeclaration(None=@n1.Gmap) {name=Gmap, nodeColor=<rgba 1 0.2109 0 1>}
  NamedRerouteDeclaration(None=@n1.AO) {name=AO, nodeColor=<rgba 1 0 0.1172 1>}
  NamedRerouteDeclaration(None=GetMaterialAttributes(None=@n1.Result) {attributeGetTypes [#1]}.Normal) {name=Normal Map, nodeColor=<rgba 1 0 0.0469 1>}
  n8 = Power(Base=@n2.A, Exp=ScalarParameter() {defaultValue=1, parameterName=EdgeIsolationPower})
  n9 = ScalarParameter() {defaultValue=3, parameterName=Boost}
  NamedRerouteDeclaration(None=Saturate(None=Multiply(A=@n8, B=@n9) {constB=2})) {name=Edge, nodeColor=<rgba 0 0.6016 1 1>}
  OneMinus(None=@n8)
 ]
}
